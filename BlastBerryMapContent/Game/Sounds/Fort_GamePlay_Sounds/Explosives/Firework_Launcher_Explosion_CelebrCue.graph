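# Sound cue for a celebratory firework launcher explosion, mixing a close blast with a far-carrying distant firework layer after a short random delay.
# Each layer picks one of three waves, varies pitch, and has its own sound class and attenuation override.
# The close layer uses logarithmic falloff with low-pass filtering, while the distant layer reaches much farther.
# A concurrency override caps simultaneous instances, stopping the oldest.
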SoundCue /BlastBerryMapContent/Game/Sounds/Fort_GamePlay_Sounds/Explosives/Firework_Launcher_Explosion_CelebrCue
root=Delay_1
nodes [
 {
  id=Attenuation_5
  children [Modulator_5]
  properties {bOverrideAttenuation=true, attenuationOverrides {falloffDistance=43635.2227, stereoSpread=10000, attenuationShapeExtents {x=17887.209}}}
 }
 {
  id=Attenuation_6
  children [Random_0]
  properties {
   attenuationSettings=/Game/Sounds/Attenuation/Fort_Weapon_Fireworks_Small_Att
   bOverrideAttenuation=true
   attenuationOverrides {
    bAttenuateWithLPF=true
    distanceAlgorithm=Logarithmic
    falloffDistance=3000
    lPFFrequencyAtMax=1709.6566
    lPFRadiusMax=2528.6787
    lPFRadiusMin=2023.8364
    stereoSpread=2000
    attenuationShapeExtents {x=4000}
   }
  }
 }
 {id=Delay_1, children [Mixer_4], properties {delayMax=0.05, delayMin=0.02}}
 {id=Mixer_4, children [SoundClass_6,SoundClass_2], properties {inputVolume [1,1.1]}}
 {id=Modulator_5, children [Random_2], properties {pitchMax=1.1, pitchMin=0.9}}
 {id=Modulator_6, children [Attenuation_6], properties {pitchMax=1.1, pitchMin=0.9}}
 {id=Random_0, children [WavePlayer_0,WavePlayer_3,WavePlayer_5], properties {weights [1,1,1]}}
 {id=Random_2, children [WavePlayer_1,WavePlayer_2,WavePlayer_4], properties {weights [1,1,1]}}
 {
  id=SoundClass_2
  children [Attenuation_5]
  properties {soundClassOverride=&Fort_WeaponRanged_Fire_Distant}
 }
 {id=SoundClass_6, children [Modulator_6], properties {soundClassOverride=&Fort_WeaponRanged}}
 {id=WavePlayer_0, properties {soundWaveAssetPtr=/Game/Sounds/Fort_Weapons/Launchers/Fireworks/Firework_Launcher_Explosion_01}}
 {
  id=WavePlayer_1
  properties {soundWaveAssetPtr=/BlastberryMapContent/Game/Sounds/Fort_Weapons/Launchers/Fireworks/Fireworks_Celebrate_Distant_01}
 }
 {
  id=WavePlayer_2
  properties {soundWaveAssetPtr=/BlastberryMapContent/Game/Sounds/Fort_Weapons/Launchers/Fireworks/Fireworks_Celebrate_Distant_02}
 }
 {id=WavePlayer_3, properties {soundWaveAssetPtr=/Game/Sounds/Fort_Weapons/Launchers/Fireworks/Firework_Launcher_Explosion_02}}
 {
  id=WavePlayer_4
  properties {soundWaveAssetPtr=/BlastberryMapContent/Game/Sounds/Fort_Weapons/Launchers/Fireworks/Fireworks_Celebrate_Distant_03}
 }
 {id=WavePlayer_5, properties {soundWaveAssetPtr=/Game/Sounds/Fort_Weapons/Launchers/Fireworks/Firework_Launcher_Explosion_03}}
]
properties {
 bOverrideConcurrency=true
 maxDistance=61522.4297
 volumeMultiplier=0.7
 concurrencyOverrides {maxCount=4, resolutionRule=StopOldest, volumeScale=0.5}
}
